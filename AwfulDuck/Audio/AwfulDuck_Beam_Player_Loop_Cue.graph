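# Sound cue for the Awful Duck event's player beam float audio, routed to the gameplay events sound class with player attenuation.
# A mixer layers the float wave at full volume with a looping de-rez wave at half volume.
# An enveloper shapes the de-rez layer, fading it in and back out over several seconds.
# A concurrency override allows one instance, stopping the farthest before refusing new ones.
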SoundCue /AwfulDuck/Audio/AwfulDuck_Beam_Player_Loop_Cue
root=Mixer_0
nodes [
 {
  id=Enveloper_0
  children [WavePlayer_1]
  properties {pitchCurve {editorCurveData {keys [{time=0}]}}, volumeCurve {editorCurveData {keys [<at 0 0>,<at 3 1>,<at 8 0>]}}}
 }
 {id=Mixer_0, children [WavePlayer_0,Enveloper_0], properties {inputVolume [1,0.5]}}
 {id=WavePlayer_0, properties {soundWaveAssetPtr=&/Awful_Duck_Spotlight_Player_Float_01}}
 {id=WavePlayer_1, properties {bLooping=true, soundWaveAssetPtr=&/Awful_Duck_Spotlight_Player_Float_DeRez_Loop_01}}
]
properties {
 attenuationSettings=&/AwfulDuck_Player_Att
 bOverrideConcurrency=true
 maxDistance=10400
 soundClassObject=&Fort_Events_Gameplay
 concurrencyOverrides {maxCount=1, resolutionRule=StopFarthestThenPreventNew}
}
