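# AwfulDuck_Intro_Cue is a Sound Cue that plays the Awful Duck intro sound.
# A single wave player feeds the Awful_Duck_Intro_01 wave straight to the output, using the AwfulDuck_Att attenuation and the Fort_Events_Gameplay sound class.
# A concurrency override limits it to one instance, stopping the farthest and otherwise preventing new plays.
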SoundCue /AwfulDuck/Audio/AwfulDuck_Intro_Cue
root=WavePlayer_0
nodes [{id=WavePlayer_0, properties {soundWaveAssetPtr=&/Awful_Duck_Intro_01}}]
properties {
 attenuationSettings=&/AwfulDuck_Att
 bOverrideConcurrency=true
 maxDistance=25400
 soundClassObject=&Fort_Events_Gameplay
 volumeMultiplier=1
 concurrencyOverrides {maxCount=1, resolutionRule=StopFarthestThenPreventNew}
}
