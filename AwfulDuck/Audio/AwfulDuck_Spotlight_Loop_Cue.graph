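# Sound cue for the Awful Duck spotlight, layering switch-on, looping and transmission sounds through shared attenuation.
# A mixer blends a randomly modulated switch-on wave, the loop, and one of five random transmission waves delayed by a second.
# An Enveloper fades the loop in and raises its pitch over time.
# A concurrency override allows one instance and blocks new ones; the root mixer's second input is empty.
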SoundCue /AwfulDuck/Audio/AwfulDuck_Spotlight_Loop_Cue
root=Mixer_2
nodes [
 {id=Delay_0, children [Random_0], properties {delayMax=1, delayMin=1}}
 {
  id=Enveloper_0
  children [WavePlayer_1]
  properties {
   pitchCurve {editorCurveData {keys [{time=0}, {time=5}, <at 25 1.25>, <at 30 1.25>]}}
   volumeCurve {editorCurveData {keys [<at 0 0>,<at 3 1>,<at 15 1>,<at 16.1152 1>]}}
  }
 }
 {id=Mixer_0, children [Modulator_0,Enveloper_0,Delay_0], properties {inputVolume [1.25,0.7,1]}}
 {id=Mixer_2, children [Mixer_0,null], properties {inputVolume [1,0.5]}}
 {id=Modulator_0, children [WavePlayer_0]}
 {id=Random_0, children [WavePlayer_2,WavePlayer_3,WavePlayer_4,WavePlayer_5,WavePlayer_6], properties {weights [1,1,1,1,1]}}
 {id=WavePlayer_0, properties {soundWaveAssetPtr=&/Awful_Duck_Spotlight_On_01}}
 {id=WavePlayer_1, properties {bLooping=true, soundWaveAssetPtr=&/Awful_Duck_Spotlight_Loop_01}}
 {id=WavePlayer_2, properties {soundWaveAssetPtr=&/Awful_Duck_Spotlight_Transmission_01}}
 {id=WavePlayer_3, properties {soundWaveAssetPtr=&/Awful_Duck_Spotlight_Transmission_02}}
 {id=WavePlayer_4, properties {soundWaveAssetPtr=&/Awful_Duck_Spotlight_Transmission_03}}
 {id=WavePlayer_5, properties {soundWaveAssetPtr=&/Awful_Duck_Spotlight_Transmission_04}}
 {id=WavePlayer_6, properties {soundWaveAssetPtr=&/Awful_Duck_Spotlight_Transmission_05}}
]
properties {
 attenuationSettings=&/AwfulDuck_Att
 bOverrideConcurrency=true
 maxDistance=25400
 soundClassObject=&Fort_Gadgets_Items
 volumeMultiplier=1.5
 concurrencyOverrides {maxCount=1, resolutionRule=PreventNew}
}
