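# Sound cue for the first-person exit from the AwfulDuck player spotlight, named as a 1P exit event sound.
# It plays a single wave, the white-out riser pre-tunnel sound, straight from one wave player with no mixing or randomization.
# It routes to the gameplay events sound class and carries a very large max distance.
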SoundCue /AwfulDuck/Audio/AwfulDuck_Spotlight_Player_Exit_1P_Cue
root=WavePlayer_1
nodes [{id=WavePlayer_1, properties {soundWaveAssetPtr=&/Awful_Duck_White_Out_Riser_PreTunnel}}]
properties {maxDistance=2097152, soundClassObject=&Fort_Events_Gameplay}
