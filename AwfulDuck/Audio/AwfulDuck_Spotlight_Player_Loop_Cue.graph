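# Looping spotlight sound cue for the AwfulDuck gadget, heard only when the IsLocal branch parameter is true.
# A mixer blends the spotlight loop with a quieter-weighted base layer and a louder PlayerFlash loop.
# An enveloper over the mix ramps pitch up across 30 seconds, then fades volume to silence by 35 seconds.
# The cue uses player attenuation, a source effect chain and the gadgets sound class.
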SoundCue /AwfulDuck/Audio/AwfulDuck_Spotlight_Player_Loop_Cue
root=Branch_0
nodes [
 {id=Branch_0, children [Enveloper_1,null,null], properties {boolParameterName=IsLocal}}
 {
  id=Enveloper_1
  children [Mixer_0]
  properties {
   pitchCurve {editorCurveData {keys [
      {interpMode=RCIM_Cubic, tangentMode=RCTM_Break, time=0}
      {arriveTangent=0.2253, interpMode=RCIM_Cubic, tangentMode=RCTM_Break, time=30, value=2}
      <at 40 2>
     ]}}
   volumeCurve {editorCurveData {keys [<at 0 1>,<at 30 1>,<at 35 0>]}}
  }
 }
 {id=Mixer_0, children [WavePlayer_1,WavePlayer_0], properties {inputVolume [0.2,0.5]}}
 {id=WavePlayer_0, properties {bLooping=true, soundWaveAssetPtr=&/Awful_Duck_Spotlight_PlayerFlash_Loop_01}}
 {id=WavePlayer_1, properties {bLooping=true, soundWaveAssetPtr=&/Awful_Duck_Spotlight_Loop_01}}
]
properties {
 attenuationSettings=&/AwfulDuck_Player_Att
 maxDistance=10400
 soundClassObject=&Fort_Gadgets_Items
 sourceEffectChain=&/AwfulDuck_SourceFXChain
 volumeMultiplier=1.5
}
