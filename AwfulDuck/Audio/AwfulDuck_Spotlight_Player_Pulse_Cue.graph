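# Sound cue for the AwfulDuck spotlight player pulse, classed as event gameplay audio.
# The node graph is empty with no root, so the cue references no sound to play.
# It still carries player attenuation settings, reduced volume and a concurrency override that limits simultaneous instances by stopping the farthest, then preventing new ones.
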SoundCue /AwfulDuck/Audio/AwfulDuck_Spotlight_Player_Pulse_Cue
root=null
nodes []
properties {
 attenuationSettings=&/AwfulDuck_Player_Att
 bOverrideConcurrency=true
 maxDistance=10400
 soundClassObject=&Fort_Events_Gameplay
 volumeMultiplier=0.6
 concurrencyOverrides {maxCount=3, resolutionRule=StopFarthestThenPreventNew}
}
