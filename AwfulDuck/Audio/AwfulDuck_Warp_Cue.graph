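# Sound cue for the AwfulDuck warp, layering a white-out tunnel enter sound with a looping tunnel bed, on a first-person warp sound class.
# Each mixer pairs a modulated enter wave with a loop that an enveloper fades in over 1.5 seconds.
# A quality switch feeds the two lower levels the base waves and the two higher levels the stereo versions.
# Concurrency allows one instance per owner, stopping the oldest.
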SoundCue /AwfulDuck/Audio/AwfulDuck_Warp_Cue
root=QualityLevel_0
nodes [
 {
  id=Enveloper_0
  children [WavePlayer_3]
  properties {
   pitchCurve {editorCurveData {keys [{time=0}]}}
   volumeCurve {editorCurveData {keys [{arriveTangent=1.7389, interpMode=RCIM_Cubic, leaveTangent=1.7389, tangentMode=RCTM_User, time=0, value=0}, <at 1.5 1>]}}
  }
 }
 {
  id=Enveloper_2
  children [WavePlayer_4]
  properties {
   pitchCurve {editorCurveData {keys [{time=0}]}}
   volumeCurve {editorCurveData {keys [{arriveTangent=1.7389, interpMode=RCIM_Cubic, leaveTangent=1.7389, tangentMode=RCTM_User, time=0, value=0}, <at 1.5 1>]}}
  }
 }
 {id=Mixer_0, children [Modulator_1,Enveloper_0], properties {inputVolume [0.75,0.7]}}
 {id=Mixer_2, children [Modulator_0,Enveloper_2], properties {inputVolume [0.75,0.7]}}
 {id=Modulator_0, children [WavePlayer_0], properties {volumeMax=1, volumeMin=1}}
 {id=Modulator_1, children [WavePlayer_2], properties {volumeMax=1, volumeMin=1}}
 {id=QualityLevel_0, children [Mixer_0,Mixer_0,Mixer_2,Mixer_2]}
 {id=WavePlayer_0, properties {soundWaveAssetPtr=&/Awful_Duck_White_Out_Tunnel_Enter_Stereo_01}}
 {id=WavePlayer_2, properties {soundWaveAssetPtr=&/Awful_Duck_White_Out_Tunnel_Enter_01}}
 {id=WavePlayer_3, properties {bLooping=true, soundWaveAssetPtr=&/Awful_Duck_White_Out_Tunnel_Loop_01}}
 {id=WavePlayer_4, properties {bLooping=true, soundWaveAssetPtr=&/Awful_Duck_White_Out_Tunnel_Loop_Stereo_01}}
]
properties {
 bOverrideConcurrency=true
 maxDistance=2097152
 soundClassObject=/Game/Sounds/FortSoundClassesAndModes/Fort_LTM_BR_InGame_Warp_1P_PSM
 virtualizationMode=Disabled
 volumeMultiplier=0.5
 concurrencyOverrides {bLimitToOwner=true, maxCount=1, resolutionRule=StopOldest}
}
