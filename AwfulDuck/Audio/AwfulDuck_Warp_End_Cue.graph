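# Sound cue for the AwfulDuck warp end, playing a white-out tunnel exit sound in the Fortnite LTM in-game warp sound class.
# A quality switch sends the two lower levels to three non-stereo-named waves and the two higher levels to three stereo variants, each picked at random with equal weight.
# Overridden concurrency allows one instance per owner, stopping the oldest, with virtualization disabled and boosted volume.
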SoundCue /AwfulDuck/Audio/AwfulDuck_Warp_End_Cue
root=QualityLevel_0
nodes [
 {id=QualityLevel_0, children [Random_1,Random_1,Random_0,Random_0]}
 {id=Random_0, children [WavePlayer_1,WavePlayer_2,WavePlayer_3], properties {bShouldExcludeFromBranchCulling=true, weights [1,1,1]}}
 {id=Random_1, children [WavePlayer_5,WavePlayer_6,WavePlayer_7], properties {bShouldExcludeFromBranchCulling=true, weights [1,1,1]}}
 {id=WavePlayer_1, properties {soundWaveAssetPtr=&/Awful_Duck_White_Out_Tunnel_Exit_Stereo_01}}
 {id=WavePlayer_2, properties {soundWaveAssetPtr=&/Awful_Duck_White_Out_Tunnel_Exit_Stereo_02}}
 {id=WavePlayer_3, properties {soundWaveAssetPtr=&/Awful_Duck_White_Out_Tunnel_Exit_Stereo_03}}
 {id=WavePlayer_5, properties {soundWaveAssetPtr=&/Awful_Duck_White_Out_Tunnel_Exit_01}}
 {id=WavePlayer_6, properties {soundWaveAssetPtr=&/Awful_Duck_White_Out_Tunnel_Exit_02}}
 {id=WavePlayer_7, properties {soundWaveAssetPtr=&/Awful_Duck_White_Out_Tunnel_Exit_03}}
]
properties {
 bOverrideConcurrency=true
 maxDistance=2097152
 soundClassObject=/Game/Sounds/FortSoundClassesAndModes/Fort_LTM_BR_InGame_Warp
 virtualizationMode=Disabled
 volumeMultiplier=1.35
 concurrencyOverrides {bLimitToOwner=true, maxCount=1, resolutionRule=StopOldest}
}
