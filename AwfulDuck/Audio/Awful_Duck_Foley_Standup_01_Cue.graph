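# Sound cue for the Awful Duck foley stand-up sound, classed as Fortnite player foley.
# A single wave player feeds the Awful_Duck_Foley_Standup_01 wave straight to the output, with no randomization or mixing.
# It borrows the third-person player walk footstep attenuation asset for distance falloff.
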SoundCue /AwfulDuck/Audio/Awful_Duck_Foley_Standup_01_Cue
root=WavePlayer_0
nodes [{id=WavePlayer_0, properties {soundWaveAssetPtr=&/Awful_Duck_Foley_Standup_01}}]
properties {
 attenuationSettings=/Game/Sounds/Attenuation/Fort_Player_Footsteps_3P_Walk_Athena_Att
 maxDistance=3584
 soundClassObject=&Fort_Player_Foley
 volumeMultiplier=1
}
